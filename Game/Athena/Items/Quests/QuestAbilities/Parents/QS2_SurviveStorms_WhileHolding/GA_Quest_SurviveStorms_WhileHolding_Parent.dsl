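; Parent quest ability for surviving storm circles while holding a required item, with rarities, tags and verb set by children.
; On activation it listens for safe zone phase updates from the game state and runs StormPhaseUpdated, stopping the listener on end.
; That event gathers inventory items of the required rarities, checks them for required and excluded tags, and broadcasts the custom quest verb on a match.
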(event Ability|EventActivateAbility
  ;[0 0 0] Activate Ability - Bind custom event to when the Storm Circle updates
  (|CallParentFunction)
  (bind _output_get (Variables|Default|SetOwningPlayerPawn (CallFunction|GetAvatarasFortPlayerPawn)))
  (bind _returnvalue (GameplayEventRouter|GetEventRouter "/Script/FortniteGame.FortGameState" _output_get))
  (bind _asynctaskproxy (GameplayEventRouter|StartListeningToStatefulEvent _returnvalue "/Script/FortniteGame.FortSafeZonePhaseUpdatedEvent")
    (:then
      (Variables|Default|SetSafeZonePhaseUpdatedAsyncTask _asynctaskproxy))
    (:OnEventReceived
      _)
    (:OnSavedEventStateExists
      _))
  (CallFunction|StormPhaseUpdated))

(event Custom|StormPhaseUpdated
  ;[0 0 0] On storm phase updated, get items for rarity, then check against those to see if it's the required item
  (Utilities|Array|Clear (Variables|Default|GetPlayerItemsOfReqRarities))
  (Variables|Default|SetOwningPlayerInventoryInterface (Fortnite|GetFortPlayerControllerfromActor (Variables|Default|GetOwningPlayerPawn)))
  (CallFunction|GetItemsforRarity)
  (CallFunction|CheckHasRequiredItem))

(event Custom|GetItemsForRarity
  ;[0 0 0] Get a list of the items in the player's inventory that match a specific rarity
  (for _array_element (Variables|SetOnChildren|GetRequiredRarityTypes)
    (bind _items (InventoryManagement|GetItemsfromInventoryOwnerforRarity (Variables|Default|GetOwningPlayerInventoryInterface) _array_element))
    (if (Utilities|Array|IsNotEmpty _items)
      (Utilities|Array|AppendArray (Variables|Default|GetPlayerItemsOfReqRarities) _items))))

(event Custom|CheckHasRequiredItem
  ;[0 0 0] Check if the player has the specified item
  (Variables|Default|SetbHasItem false)
  (bind _array_element (Utilities|Array|ForEachLoopwithBreak (Variables|Default|GetPlayerItemsOfReqRarities))
    (:LoopBody
      (if (and (GameplayTags|HasAnyTags (CallFunction|GetOwnedGameplayTags(Fix) _array_element) (Variables|SetOnChildren|GetT_RequiredItemTags)) (not (GameplayTags|HasAnyTags (CallFunction|GetOwnedGameplayTags(Fix) _array_element) (Variables|SetOnChildren|GetT_ExcludedWeaponTags) true)))
        (Variables|Default|SetbHasItem true)))
    (:Completed
      _))
  (if (Variables|Default|GetbHasItem)
    (CustomVerb|BroadcastCustomVerbMessage (CustomVerb|MakeCustomVerbMessage (Variables|Default|GetOwningPlayerPawn) (Variables|SetOnChildren|GetVerb%20Name)))))

(event Ability|EventOnEndAbility
  (params (bWasCancelled :type Boolean))
  ;[0 0 0] End ability - clean up event binds
  (bind _safezonephaseupdatedasynctask (Variables|Default|GetSafeZonePhaseUpdatedAsyncTask)
    (:then
      (GameplayEventRouter|StopListeningtoStatefulEvent _safezonephaseupdatedasynctask))
    (:else)))

(fn GetOwnedGameplayTags(Fix)
  (params (Item :type Fort%20World%20Item%20Object%20Reference))
  (return (GameplayTags|GetOwnedGameplayTags Item :!class /Script/GameplayTags.GameplayTagAssetInterface)))
